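# Character-parent material for the GallonBag pickaxe that flashes a tinted, masked emissive glow on a reactivity event.
# The glow multiplies intensity, tint and a powered bleeding-mask red channel.
# Time since the ReactivityTime parameter, divided by fade seconds and saturated, lerps emissive from that glow back to normal, replacing base emissive rather than adding to it.
# The override sits between the shared 2023 PreFX and PostFX material functions.
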Material /BRCosmetics/Weapons/FORT_Melee/Pickaxe_GallonBag/Materials/MI_Pickaxe_GallonBag_Reactivity
graph {
 properties {
  bUseMaterialAttributes=true
  bUsedWithMeshDeformer=true
  bUsedWithSkeletalMesh=true
  pixelDepthOffsetMode=PDOM_Legacy
  shadingModel=MSM_FromMaterialExpression
  usedShadingModels=Default Lit | Subsurface | Clear Coat | Subsurface Profile | Cloth
 }
 nodes [
  n1 = MaterialFunctionCall() {materialFunction=&MF_CharacterParent_2023_PreFX}
  n2 = GetMaterialAttributes(None=@n1.Result) {attributeGetTypes [#1]}
  n3 = TextureSampleParameter2D() {
   parameterName=T_BleedingMask
   samplerType=SAMPLERTYPE_Masks
   texture=/BRCosmetics/Weapons/FORT_Melee/Pickaxe_GallonBag/Textures/T_EmissiveBleedingMask
  }
  n4 = Multiply(
   A=Multiply(
    A=ScalarParameter() {defaultValue=30, group=Reactivity, parameterName=HighEmissive}
    B=VectorParameter() {group=Reactivity, parameterName=ReactEmissive_Tint, defaultValue=<rgba 1 0 0 1>}.RGB
   )
   B=Power(Base=@n3.R)
  )
  n5 = Subtract(A=Time(), B=ScalarParameter() {group=Reactivity, parameterName=ReactivityTime})
  n6 = ScalarParameter() {defaultValue=3, group=Reactivity, parameterName=FadeTime_Second}
  n7 = NamedRerouteDeclaration(None=Saturate(None=Divide(A=@n5, B=@n6))) {name=ReactAlpha, nodeColor=<rgba 0.0312 1 0 1>}
  MaterialFunctionCall(
   Material=SetMaterialAttributes(MaterialAttributes=@n2.MaterialAttributes, Emissive Color=LinearInterpolate(A=@n4, B=@n2.Emissive Color, Alpha=@n7)) {attributeSetTypes [#1]}
   M=@n1.M
   Pre Skinned Local Position=@n1.M
   Distance Blend=@n1.M
   Fresnel=@n1.M
  ) {materialFunction=&MF_CharacterParent_2023_PostFX}
 ]
}
