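# Physics asset for the GallonBag pickaxe, a melee weapon with two dangling rope chains (A and B) that can be simulated as ragdoll-style links.
# Each rope is four capsule bodies chained from the two-handed handle, with swing limited and twist locked at every joint; the first joint of rope B swings slightly wider than that of A.
# The handle also joins the root and a wide hilt-base capsule through constraints with free swing and twist.
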PhysicsAsset /BRCosmetics/Weapons/FORT_Melee/Pickaxe_GallonBag/Meshes/GallonBag_Axe_Physics
bodies {
 A_rope_01 [{shapeType=Capsule, length=3, radius=2, shapeName=A_rope_01_capsule, rotation=<rot 0 90 90>}]
 A_rope_02 [{shapeType=Capsule, length=2, radius=1.5, shapeName=A_rope_02_capsule, center=<1.3886 0 0>, rotation=<rot 0 90 90>}]
 A_rope_03 [{shapeType=Capsule, length=4, radius=2, shapeName=A_rope_03_capsule, center=<2.6233 0 0>, rotation=<rot 0 90 90>}]
 A_rope_04 [{shapeType=Capsule, length=8, radius=2.5, shapeName=A_rope_04_capsule, center=<5.1368 0 0>, rotation=<rot -0.0001 90.0001 93.407>}]
 B_rope_01 [{shapeType=Capsule, length=3, radius=2, shapeName=B_rope_01_capsule, rotation=<rot 0 90 90>}]
 B_rope_02 [{shapeType=Capsule, length=2, radius=1.5, shapeName=B_rope_02_capsule, center=<1.3886 0 0>, rotation=<rot 0 90 90>}]
 B_rope_03 [{shapeType=Capsule, length=4, radius=2, shapeName=B_rope_03_capsule, center=<2.6233 0 0>, rotation=<rot 0 90 90>}]
 B_rope_04 [{shapeType=Capsule, length=8, radius=2.5, shapeName=B_rope_04_capsule, center=<4.3173 0 0>, rotation=<rot -0.0001 90.0001 95.1211>}]
 Melee_TwoHanded_Handle [{shapeType=Capsule, length=1, radius=1, shapeName=Melee_TwoHanded_Handle_capsule}]
 b_AR_Root [{shapeType=Capsule, length=0.505, radius=0.505, shapeName=b_AR_Root_capsule, rotation=<rot 0 -90 0>}]
 hilt_base [{shapeType=Capsule, length=1.5, radius=6.3, shapeName=hilt_base_capsule, center=<0 0 -2.2672>, rotation=<rot 0 -90 0>}]
}
constraints [
 {bone1Name=A_rope_02, bone2Name=A_rope_01, swing1LimitDegrees=30, swing2LimitDegrees=30, twistLimitDegrees=45, twistMotion=Locked}
 {bone1Name=A_rope_03, bone2Name=A_rope_02, swing1LimitDegrees=30, swing2LimitDegrees=30, twistLimitDegrees=45, twistMotion=Locked}
 {bone1Name=B_rope_02, bone2Name=B_rope_01, swing1LimitDegrees=30, swing2LimitDegrees=30, twistLimitDegrees=45, twistMotion=Locked}
 {bone1Name=B_rope_03, bone2Name=B_rope_02, swing1LimitDegrees=30, swing2LimitDegrees=30, twistLimitDegrees=45, twistMotion=Locked}
 {
  bone1Name=A_rope_01
  bone2Name=Melee_TwoHanded_Handle
  swing1LimitDegrees=32
  swing2LimitDegrees=32
  twistLimitDegrees=45
  twistMotion=Locked
 }
 {
  bone1Name=B_rope_01
  bone2Name=Melee_TwoHanded_Handle
  swing1LimitDegrees=40
  swing2LimitDegrees=40
  twistLimitDegrees=45
  twistMotion=Locked
 }
 {
  bone1Name=Melee_TwoHanded_Handle
  bone2Name=b_AR_Root
  swing1LimitDegrees=45
  swing1Motion=Free
  swing2LimitDegrees=45
  swing2Motion=Free
  twistLimitDegrees=45
  twistMotion=Free
 }
 {bone1Name=A_rope_04, bone2Name=A_rope_03, swing1LimitDegrees=30, swing2LimitDegrees=30, twistLimitDegrees=45, twistMotion=Locked}
 {bone1Name=B_rope_04, bone2Name=B_rope_03, swing1LimitDegrees=30, swing2LimitDegrees=30, twistLimitDegrees=45, twistMotion=Locked}
 {
  bone1Name=hilt_base
  bone2Name=Melee_TwoHanded_Handle
  swing1LimitDegrees=45
  swing1Motion=Free
  swing2LimitDegrees=45
  swing2Motion=Free
  twistLimitDegrees=45
  twistMotion=Free
 }
]
