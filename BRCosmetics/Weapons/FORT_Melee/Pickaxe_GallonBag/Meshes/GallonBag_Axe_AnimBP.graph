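# Animation Blueprint for the GallonBag pickaxe that adds secondary physics motion to its dangling wire and rope bones.
# The input pose runs through a RigidBody node, enabled only outside the front end, then additive bone offsets on the ropes, and is cached.
# Nine chained world-space AnimDynamics nodes then add spring motion to the wire and two ropes.
# The update event also sets gravity-override and front-end alpha variables, which the AnimGraph does not visibly read.
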AnimBlueprint /BRCosmetics/Weapons/FORT_Melee/Pickaxe_GallonBag/Meshes/GallonBag_Axe_AnimBP
target_skeleton=&/GallonBag_Axe_Skeleton
graphs [
 AnimGraph {nodes [
   n1 = Anim Dynamics - Bone: dyn_wire_01!(
    ComponentPose=Local To Component(LocalPose=Use cached pose 'RigidBody / Adjusments'())
    Alpha=1
    GravityOverride:Vector="0,0,0"
    ExternalForce="0,0,0"
   ) -> (Pose) {
    angularDampingOverride=1
    angularSpringConstant=33000
    bAngularSpring=true
    bOverrideAngularDamping=true
    bUseGravityOverride=true
    gravityScale=0.35
    simulationSpace=World
    alphaBoolBlend {blendOption=Linear}
    alphaScaleBiasClamp {bInterpResult=true}
    boundBone {boneName=dyn_wire_01}
    componentLinearAccScale=<1 1 1>
    componentLinearVelScale=<5 5 5>
    gravityOverride {z=-980}
    physicsBodyDefinitions [{
      boundBone {boneName=dyn_wire_01}
      boxExtents=<20 30 30>
      constraintSetup {angularLimitsMax=<0 -5 7>, angularLimitsMin=<0 -5 -7>, angularTarget=<1 0 0>}
      localJointOffset=<0.5 0 0>
     }]
    relativeSpaceBone {boneName=Melee_TwoHanded_Handle}
   }
   n2 = Anim Dynamics - Bone: dyn_wire_02!(ComponentPose=@n1, Alpha=1, GravityOverride:Vector="0,0,0", ExternalForce="0,0,0") -> (Pose) {
    angularDampingOverride=1
    angularSpringConstant=33000
    bAngularSpring=true
    bOverrideAngularDamping=true
    bUseGravityOverride=true
    gravityScale=0.35
    simulationSpace=World
    alphaBoolBlend {blendOption=Linear}
    alphaScaleBiasClamp {bInterpResult=true}
    boundBone {boneName=dyn_wire_02}
    componentLinearAccScale=<1 1 1>
    componentLinearVelScale=<5 5 5>
    gravityOverride {z=-980}
    physicsBodyDefinitions [{
      boundBone {boneName=dyn_wire_02}
      boxExtents=<20 30 30>
      constraintSetup {angularLimitsMax=<0 7 7>, angularLimitsMin=<0 -7 -7>, angularTarget=<1 0 0>}
      localJointOffset=<-0.5 0 0>
     }]
    relativeSpaceBone {boneName=Melee_TwoHanded_Handle}
   }
   n3 = Anim Dynamics - Bone: dyn_wire_03!(ComponentPose=@n2, Alpha=1, GravityOverride:Vector="0,0,0", ExternalForce="0,0,0") -> (Pose) {
    angularDampingOverride=1
    angularSpringConstant=33000
    bAngularSpring=true
    bOverrideAngularDamping=true
    bUseGravityOverride=true
    gravityScale=0.35
    simulationSpace=World
    alphaBoolBlend {blendOption=Linear}
    alphaScaleBiasClamp {bInterpResult=true}
    boundBone {boneName=dyn_wire_03}
    componentAppliedLinearAccClamp {y=79153.0949}
    componentLinearAccScale=<1 1 1>
    componentLinearVelScale=<5 5 5>
    gravityOverride {z=-980}
    physicsBodyDefinitions [{
      boundBone {boneName=dyn_wire_03}
      boxExtents=<20 30 30>
      constraintSetup {angularLimitsMax=<0 10 10>, angularLimitsMin=<0 -10 -10>, angularTarget=<1 0 0>}
      localJointOffset=<0.5 0 0>
     }]
    relativeSpaceBone {boneName=Melee_TwoHanded_Handle}
   }
   n4 = Anim Dynamics - Bone: A_rope_01!(ComponentPose=@n3, bAlphaBoolEnabled:bool=True, GravityOverride:Vector="0,0,0", ExternalForce="0,0,0") -> (Pose) {
    alphaInputType=Bool
    angularDampingOverride=1
    angularSpringConstant=15000
    bAngularSpring=true
    bOverrideAngularDamping=true
    bUseGravityOverride=true
    gravityScale=0.35
    simulationSpace=World
    alphaBoolBlend {blendOption=Linear}
    boundBone {boneName=A_rope_01}
    componentLinearAccScale=<1 1 1>
    componentLinearVelScale=<1 1 1>
    gravityOverride {z=-980}
    physicsBodyDefinitions [{
      boundBone {boneName=A_rope_01}
      boxExtents=<20 30 30>
      constraintSetup {angularLimitsMax=<5 5 10>, angularLimitsMin=<-5 5 -10>, angularTarget=<1 0 0>}
      localJointOffset=<4 0 0>
     }]
    relativeSpaceBone {boneName=Melee_TwoHanded_Handle}
   }
   n5 = Anim Dynamics - Bone: A_rope_02!(ComponentPose=@n4, bAlphaBoolEnabled:bool=True, GravityOverride:Vector="0,0,0", ExternalForce="0,0,0") -> (Pose) {
    alphaInputType=Bool
    angularDampingOverride=1
    angularSpringConstant=15000
    bAngularSpring=true
    bOverrideAngularDamping=true
    bUseGravityOverride=true
    gravityScale=0.35
    simulationSpace=World
    alphaBoolBlend {blendOption=Linear}
    boundBone {boneName=A_rope_02}
    componentLinearAccScale=<1 1 1>
    componentLinearVelScale=<1 1 1>
    gravityOverride {z=-980}
    physicsBodyDefinitions [{
      boundBone {boneName=A_rope_02}
      boxExtents=<20 30 30>
      constraintSetup {angularLimitsMax=<5 15 10>, angularLimitsMin=<-5 -5 -10>, angularTarget=<1 0 0>}
      localJointOffset=<4 0 0>
     }]
    relativeSpaceBone {boneName=Melee_TwoHanded_Handle}
   }
   n6 = Anim Dynamics - Bone: A_rope_03!(ComponentPose=@n5, bAlphaBoolEnabled:bool=True, GravityOverride:Vector="0,0,0", ExternalForce="0,0,0") -> (Pose) {
    alphaInputType=Bool
    angularDampingOverride=1
    angularSpringConstant=15000
    bAngularSpring=true
    bOverrideAngularDamping=true
    bUseGravityOverride=true
    gravityScale=0.35
    simulationSpace=World
    alphaBoolBlend {blendOption=Linear}
    boundBone {boneName=A_rope_03}
    componentLinearAccScale=<1 1 1>
    componentLinearVelScale=<1 1 1>
    gravityOverride {z=-980}
    physicsBodyDefinitions [{
      boundBone {boneName=A_rope_03}
      boxExtents=<20 30 30>
      constraintSetup {angularLimitsMax=<5 15 10>, angularLimitsMin=<-5 -5 -10>, angularTarget=<1 0 0>}
      localJointOffset=<4 0 0>
     }]
    relativeSpaceBone {boneName=Melee_TwoHanded_Handle}
   }
   n7 = Anim Dynamics - Bone: B_rope_01!(ComponentPose=@n6, bAlphaBoolEnabled:bool=True, GravityOverride:Vector="0,0,0", ExternalForce="0,0,0") -> (Pose) {
    alphaInputType=Bool
    angularDampingOverride=1
    angularSpringConstant=15000
    bAngularSpring=true
    bOverrideAngularDamping=true
    bUseGravityOverride=true
    gravityScale=0.35
    simulationSpace=World
    alphaBoolBlend {blendOption=Linear}
    boundBone {boneName=B_rope_01}
    componentLinearAccScale=<1 1 1>
    componentLinearVelScale=<1 1 1>
    gravityOverride {z=-980}
    physicsBodyDefinitions [{
      boundBone {boneName=B_rope_01}
      boxExtents=<30 30 30>
      constraintSetup {angularLimitsMax=<15 15 10>, angularLimitsMin=<-15 -15 -10>, angularTarget=<1 0 0>}
      localJointOffset=<-2 1 0>
     }]
    relativeSpaceBone {boneName=Melee_TwoHanded_Handle}
   }
   n8 = Anim Dynamics - Bone: B_rope_02!(ComponentPose=@n7, bAlphaBoolEnabled:bool=True, GravityOverride:Vector="0,0,0", ExternalForce="0,0,0") -> (Pose) {
    alphaInputType=Bool
    angularDampingOverride=1
    angularSpringConstant=15000
    bAngularSpring=true
    bOverrideAngularDamping=true
    bUseGravityOverride=true
    gravityScale=0.35
    simulationSpace=World
    alphaBoolBlend {blendOption=Linear}
    boundBone {boneName=B_rope_02}
    componentLinearAccScale=<1 1 1>
    componentLinearVelScale=<1 1 1>
    gravityOverride {z=-980}
    physicsBodyDefinitions [{
      boundBone {boneName=B_rope_02}
      boxExtents=<30 30 30>
      constraintSetup {angularLimitsMax=<10 15 10>, angularLimitsMin=<-15 -15 -10>, angularTarget=<1 0 0>}
      localJointOffset=<-2 1 0>
     }]
    relativeSpaceBone {boneName=Melee_TwoHanded_Handle}
   }
   n9 = Anim Dynamics - Bone: B_rope_03!(ComponentPose=@n8, bAlphaBoolEnabled:bool=True, GravityOverride:Vector="0,0,0", ExternalForce="0,0,0") -> (Pose) {
    alphaInputType=Bool
    angularDampingOverride=1
    angularSpringConstant=15000
    bAngularSpring=true
    bOverrideAngularDamping=true
    bUseGravityOverride=true
    gravityScale=0.35
    simulationSpace=World
    alphaBoolBlend {blendOption=Linear}
    boundBone {boneName=B_rope_03}
    componentLinearAccScale=<1 1 1>
    componentLinearVelScale=<1 1 1>
    gravityOverride {z=-980}
    physicsBodyDefinitions [{
      boundBone {boneName=B_rope_03}
      boxExtents=<30 30 30>
      constraintSetup {angularLimitsMax=<10 15 10>, angularLimitsMin=<-15 -15 -10>, angularTarget=<1 0 0>}
      localJointOffset=<-2 1 0>
     }]
    relativeSpaceBone {boneName=Melee_TwoHanded_Handle}
   }
   Output Pose(Result=Component To Local(ComponentPose=@n9)) {layerGroup=None}
   AnimGraphNode_SubInput_0 = Input Pose()
   n10 = RigidBody!(
    ComponentPose=Local To Component(LocalPose=@AnimGraphNode_SubInput_0)
    bAlphaBoolEnabled:bool=Get is_BR_Level() -> (is_BR_Level:bool)
    OverrideWorldGravity:Vector="0,0,-980"
    ExternalForce
   ) -> (Pose) {
    alphaInputType=Bool
    bOverrideWorldGravity=true
    overridePhysicsAsset=&/GallonBag_Axe_Physics
    simulationSpace=BaseBoneSpace
    alphaBoolBlend {blendOption=Linear}
    alphaScaleBiasClamp {bInterpResult=true}
    baseBoneRef {boneName=Melee_TwoHanded_Handle}
    componentAppliedLinearAccClamp=<3000 3000 3000>
    componentLinearAccScale=<0.3 0.3 0.3>
    componentLinearVelScale=<0.3 0.3 0.3>
    simSpaceSettings {maxAngularAcceleration=1000, maxAngularVelocity=1000, maxLinearAcceleration=1000, maxLinearVelocity=1000, worldAlpha=0.85}
   }
   n11 = "Transform (Modify) Bone - Bone: ropes_base"!(ComponentPose=@n10, bAlphaBoolEnabled:bool=True, Translation="0,0,0", Rotation="18,50,0", Scale="1,1,1") -> (Pose) {
    alphaInputType=Bool
    rotationMode=BMM_Additive
    rotationSpace=BCS_BoneSpace
    alphaBoolBlend {blendOption=Linear}
    boneToModify {boneName=ropes_base}
    rotation {pitch=18, yaw=50}
   }
   n12 = "Transform (Modify) Bone - Bone: A_rope_Base"!(ComponentPose=@n11, bAlphaBoolEnabled:bool=True, Translation="0,0,0", Rotation="10,0,0", Scale="1,1,1") -> (Pose) {
    alphaInputType=Bool
    rotationMode=BMM_Additive
    rotationSpace=BCS_BoneSpace
    alphaBoolBlend {blendOption=Linear}
    boneToModify {boneName=A_rope_Base}
    rotation {pitch=10}
   }
   n13 = "Transform (Modify) Bone - Bone: B_rope_02"!(ComponentPose=@n12, bAlphaBoolEnabled:bool=True, Translation="1,0,0", Rotation="16.5,0,0", Scale="1,1,1") -> (Pose) {
    alphaInputType=Bool
    rotationMode=BMM_Additive
    rotationSpace=BCS_BoneSpace
    translationMode=BMM_Additive
    translationSpace=BCS_BoneSpace
    alphaBoolBlend {blendOption=Linear}
    boneToModify {boneName=B_rope_02}
    rotation {pitch=16.5}
    translation {x=1}
   }
   n14 = "Transform (Modify) Bone - Bone: B_rope_Base"!(ComponentPose=@n13, bAlphaBoolEnabled:bool=True, Translation="0,0,0", Rotation="1.5,0,0", Scale="1,1,1") -> (Pose) {
    alphaInputType=Bool
    rotationMode=BMM_Additive
    rotationSpace=BCS_BoneSpace
    translationMode=BMM_Additive
    translationSpace=BCS_BoneSpace
    alphaBoolBlend {blendOption=Linear}
    boneToModify {boneName=B_rope_Base}
    rotation {pitch=1.5}
   }
   n15 = "Transform (Modify) Bone - Bone: A_rope_02"!(ComponentPose=@n14, bAlphaBoolEnabled:bool=True, Translation="0,0,0", Rotation="10,0,0", Scale="1,1,1") -> (Pose) {
    alphaInputType=Bool
    rotationMode=BMM_Additive
    rotationSpace=BCS_BoneSpace
    alphaBoolBlend {blendOption=Linear}
    boneToModify {boneName=A_rope_02}
    rotation {pitch=10}
   }
   Save cached pose 'RigidBody / Adjusments'(Pose=Component To Local(ComponentPose=@n15)) {cacheName=RigidBody / Adjusments}
  ]}
 EventGraph {nodes [
   Event BlueprintUpdateAnimation()
   > n1 = Sequence()
   > n2 = Gravity Override State Function DEPRECATED(CharacterPartAnimInstance=Self-Reference(), GameTimeInSeconds=GetGameTimeInSeconds(), GravityOverrideParams=Get Gravity Override Params())
   > Set GravityOverride(GravityOverride=@n2)
   n3 = Branch(execute=@n1.then_1, Condition=Get bIsInFrontEnd())
   > Set is_BR_Level(is_BR_Level:bool=false) -> (Output_Get:bool)
   > Set Alpha is FrontEnd Dynamics(Alpha is FrontEnd Dynamics:double=0.7) -> (Output_Get:double)
   Set is_BR_Level(execute=@n3.else, is_BR_Level=true)
   > Set Alpha is FrontEnd Dynamics(Alpha is FrontEnd Dynamics=1)
  ]}
]
variables [
 Alpha is FrontEnd Dynamics {type=number}
 Gravity Override Params {type=GravityOverrideParamsStruct}
 GravityOverride {type=Vector}
 is_BR_Level {type=boolean}
]
